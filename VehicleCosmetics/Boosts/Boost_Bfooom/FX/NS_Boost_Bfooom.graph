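# Niagara system for the Bfooom vehicle boost cosmetic, streaming a trail and sparks backward from the car's boost sockets.
# A multi-renderer ribbon emitter spawns one trail per filtered socket, while two sprite emitters partition particles across those sockets on the skeletal mesh.
# The spark emitters merge several effects through spawn groups, with SubUV animation and per-group tuning.
# Each emitter applies painted-variant color overrides, some with extra per-renderer color channels.
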NiagaraSystem /VehicleCosmetics/Boosts/Boost_Bfooom/FX/NS_Boost_Bfooom
dependencies {
 usedDataInterfaces [
  &NiagaraDataInterfaceCurve
  &NiagaraDataInterfaceSpriteRendererInfo
  &NiagaraDataInterfaceArrayFloat
  &NiagaraDataInterfaceVector2DCurve
  &NiagaraDataInterfaceSkeletalMesh
 ]
 usedDynamicInputs [/VehicleCosmetics/FX/Niagara/Modules/MergeEmitters/VC_VectorBySpawnGroup, /VehicleCosmetics/FX/Niagara/Modules/VC_Mask_Floats]
 usedModules [
  &self:SystemSpawnScript.SystemScriptSource.SystemScriptGraph.NiagaraNodeAssignment_0.SetVariables_#1
  &SystemState
 ]
}
emitters {
 RibbonGroup001 {
  emitterUpdateScript [
   &EmitterState(Inactive Response=::NewEnumerator1, Life Cycle Mode=::NewEnumerator1, Loop Duration=1, UseLoopDelay=false)
   &VC_FilteredSocketCount()
   &VC_RateQualityHandler()
   &VC_SpawnRibbon_v1(Number of Ribbons=$Emitter.PortCount, Spawn Time=-1, SpawnRate=60)
  ]
  particleSpawnScript [
   &VC_PaintedVariantsOverride(
    Black=<rgba 0.005 0.005 0.005 4>
    BurntSienna=<rgba 3 0.2158 0 2>
    Crimson=<rgba 3 0.25 0.3245 1>
    DefaultColor=<rgba 0.1406 0.6172 3 1>
    ForestGreen=<rgba 0 2 0.1011 1>
    Grey=<rgba 0.2865 0.2865 0.2865 2>
    Lime=<rgba 1.7649 3 0 1>
    Orange=<rgba 1.03 0.8174 0 1>
    Pink=<rgba 3 0.5312 2.1023 1>
    Purple=<rgba 0.8096 0 3 1>
    R1 Black:LinearColor=<rgba 0 0 0 3>
    R1 BurntSienna:LinearColor=<rgba 2 0.4231 0 1>
    R1 Crimson:LinearColor=<rgba 3 0 0.05 1>
    R1 DefaultColor=<rgba 0 0.8889 4 1>
    R1 ForestGreen:LinearColor=<rgba 0 0.9531 0.0482 1>
    R1 Grey:LinearColor=<rgba 0.0365 0.0365 0.0365 2>
    R1 Lime:LinearColor=<rgba 2.5 5 0 1>
    R1 Orange:LinearColor=<rgba 5 0.3528 0 1>
    R1 Pink:LinearColor=<rgba 3 0.15 1.9636 1>
    R1 Saffron:LinearColor=<rgba 1 0.7667 0 1>
    R1 SkyBlue:LinearColor=<rgba 0 0.6654 1 1>
    R2 Black:LinearColor=<rgba 0 0 0 1>
    R2 BurntSienna:LinearColor=<rgba 0.1823 0.0564 0.0223 1>
    R2 Crimson:LinearColor=<rgba 1 0 0.0271 1>
    R2 DefaultColor:LinearColor=<rgba 0 0.3079 0.65 20>
    R2 ForestGreen:LinearColor=<rgba 0 0.8 0.0404 1>
    R2 Grey:LinearColor=<rgba 0.6302 0.6302 0.6302 1>
    R2 Lime:LinearColor=<rgba 0.3038 0.349 0 1>
    R2 Orange:LinearColor=<rgba 1.3 0.4534 0 1>
    R2 Pink:LinearColor=<rgba 1 0.1198 0.6799 1>
    R2 Purple:LinearColor=<rgba 0.75 0.5 1 1>
    R2 Saffron:LinearColor=<rgba 1 1 0 1>
    R2 SkyBlue:LinearColor=<rgba 0 1 1 1>
    R2 TitaniumWhite:LinearColor=<rgba 1.5 1.5 1.5 1>
    R3 Black:LinearColor=<rgba 0.0417 0.0417 0.0417 3>
    R3 Crimson:LinearColor=<rgba 0.2656 0 0.0072 1>
    R3 DefaultColor:LinearColor=<rgba 0 0.3693 2.2156 1>
    R3 ForestGreen:LinearColor=<rgba 0 0.1302 0.0066 1>
    R3 Grey:LinearColor=<rgba 0.099 0.099 0.099 5>
    R3 Lime:LinearColor=<rgba 0.2083 0.4167 0 1>
    R3 Orange:LinearColor=<rgba 0.8522 0.2609 0 1>
    R3 Pink:LinearColor=<rgba 0.5052 0.0605 0.3435 1>
    R3 Purple:LinearColor=<rgba 0.3423 0 0.875 1>
    R3 Saffron:LinearColor=<rgba 0.6927 0.5845 0 1>
    R3 SkyBlue:LinearColor=<rgba 0 0.6654 1 1>
    R3 TitaniumWhite:LinearColor=<rgba 0.3594 0.3594 0.3594 1>
    Saffron=<rgba 3 1.0875 0 1>
    SkyBlue=<rgba 0.75 2.2471 3 1>
    TitaniumWhite=<rgba 1 1 1 2>
    Use R1 Painted Overrides=true
    Use R2 Painted Overrides=true
    Use R3 Painted Overrides=true
   )
   &VC_InitializeRibbon_v1(Color=$Particles.Color, Enable Skeletal Position=true, Lifetime=1.5, Offset Amount=1, Ribbon Width=40)
   &AddVelocity(Velocity=<-1500 0 0>) {module=AddVelocity001}
   /SetVariables(Particles.Ribbon2=85) {isSetParameters=true}
  ]
  particleUpdateScript [
   &ParticleState()
   &SolveForcesAndVelocity(Clamp Velocity=false, Limit Acceleration=false)
   &VC_UpdateRibbon_v1(Enable Skeletal Position=true)
   /SetVariables() {isSetParameters=true}
   /SetVariables(
    Particles.ScaleColorCoreLighting_R1:LinearColor=&VC_DynamicScaleColor(
     Color to Scale:LinearColor=$Particles.R1 Color
     RGB Scale:Vector3f=&VectorFromFloat(Value=30)
     Alpha Scale:NiagaraFloat=&FloatFromCurve(FloatCurve=DataInterface() {propertyValues {Curve {keys [<at 0 0 RCIM_Cubic>, <at 0.2028 1 RCIM_Cubic>, {interpMode=RCIM_Cubic, leaveTangent=-2.4222, time=0.7671, value=0}]}}})
    )
    Particles.ScaleColorMasDark_R2:LinearColor=&VC_DynamicScaleColor(
     Color to Scale:LinearColor=$Particles.R2 Color
     RGB Scale:Vector3f=&VectorFromFloat(Value=0.02)
     Alpha Scale:NiagaraFloat=&MakeCustomFloatFromBool(Bool=$User.bIsPainted, True Float=20, False Float=1)
    )
    Particles.ScalecolorTrailRibbon_Rate0:LinearColor=&VC_DynamicScaleColor(
     Color to Scale:LinearColor=$Particles.Color
     RGB Scale:Vector3f=&VectorFromFloat(Value=0.1)
     Alpha Scale:NiagaraFloat=1
    )
   ) {isSetParameters=true}
  ]
  rendererClasses [&NiagaraRibbonRendererProperties]
  renderers [
   {rendererClass=&NiagaraRibbonRendererProperties}
   {rendererClass=&NiagaraRibbonRendererProperties, rendererIndex=1}
   {rendererClass=&NiagaraRibbonRendererProperties, rendererIndex=2}
   {rendererClass=&NiagaraRibbonRendererProperties, rendererIndex=3}
  ]
 }
 SparksGroup001 {
  emitterUpdateScript [
   &EmitterState(Inactive Response=::NewEnumerator1, Life Cycle Mode=::NewEnumerator1, Loop Duration=1, UseLoopDelay=false)
   &VC_FilteredSocketCount()
   &SpawnRate(SpawnRate=20, Use Spawn Probability=false)
   &SpawnRate(SpawnGroup=1, SpawnRate=30, Use Spawn Probability=false) {module=SpawnRate003}
   &SpawnRate(SpawnGroup=2, SpawnRate=30, Use Spawn Probability=false) {module=SpawnRate004}
  ]
  particleSpawnScript [
   &VC_PaintedVariantsOverride(
    Black=<rgba 0.0833 0.0833 0.0833 1>
    BurntSienna=<rgba 1 0.2116 0 1>
    Crimson=<rgba 1 0 0.0271 1>
    DefaultColor=<rgba 0 2.2222 5 1>
    ForestGreen=<rgba 0 1 0.0505 1>
    Grey=<rgba 0.8 0.8 0.8 1>
    Lime=<rgba 0.25 0.5 0 1>
    Orange=<rgba 1 0.0833 0 1>
    Pink=<rgba 1 0.1198 0.6799 1>
    Purple=<rgba 0.5 0 1 1>
    R1 Black:LinearColor=<rgba 0.0365 0.0365 0.0365 1>
    R1 BurntSienna:LinearColor=<rgba 1 0.2116 0 1>
    R1 Crimson:LinearColor=<rgba 0.5 0 0.0136 1>
    R1 DefaultColor=<rgba 0 0.4444 1 1>
    R1 ForestGreen:LinearColor=<rgba 0 1 0.0505 1>
    R1 Grey:LinearColor=<rgba 0.0469 0.0469 0.0469 1>
    R1 Lime:LinearColor=<rgba 0.25 0.5 0 1>
    R1 Orange:LinearColor=<rgba 1 0.0833 0 1>
    R1 Pink:LinearColor=<rgba 1 0.1198 0.6799 1>
    R1 Purple:LinearColor=<rgba 0.25 0 0.5 1>
    R1 Saffron:LinearColor=<rgba 1 0.7667 0 1>
    R1 SkyBlue:LinearColor=<rgba 0 0.6654 1 1>
    R1 TitaniumWhite:LinearColor=<rgba 1 1 1 1>
    R2 Black:LinearColor=<rgba 0 0 0 1>
    R2 BurntSienna:LinearColor=<rgba 1 0.2116 0 1>
    R2 Crimson:LinearColor=<rgba 0.5 0 0.0136 1>
    R2 DefaultColor:LinearColor=<rgba 0 0.1276 0.5104 1>
    R2 ForestGreen:LinearColor=<rgba 0 1 0.0505 1>
    R2 Grey:LinearColor=<rgba 0.1615 0.1615 0.1615 1>
    R2 Lime:LinearColor=<rgba 0.25 0.5 0 1>
    R2 Orange:LinearColor=<rgba 1 0.0833 0 1>
    R2 Pink:LinearColor=<rgba 1 0.1198 0.6799 1>
    R2 Purple:LinearColor=<rgba 0.25 0 0.5 1>
    R2 Saffron:LinearColor=<rgba 1 0.7667 0 1>
    R2 SkyBlue:LinearColor=<rgba 0 0.6654 1 1>
    R2 TitaniumWhite:LinearColor=<rgba 1 1 1 1>
    Saffron=<rgba 1 0.7667 0 1>
    SkyBlue=<rgba 0 0.6654 1 1>
    TitaniumWhite=<rgba 1 1 1 1>
    Use R1 Painted Overrides=true
    Use R2 Painted Overrides=true
   )
   /SetVariables(Particles.Visivily_tag:NiagaraInt32=&SpawnGroup()) {isSetParameters=true, module=SetVariables_#2}
   &InitializeParticle(
    Color=$Particles.Color
    Lifetime=/VehicleCosmetics/FX/Niagara/Modules/MergeEmitters/VC_FloatBySpawnGroup(
     SpawnGroup or ExecIndex:NiagaraInt32=0
     CustomInt:NiagaraInt32=0
     MaximumSpawnGroups:NiagaraInt32=3
     Float SG 0:NiagaraFloat=&RandomRangeFloat(Minimum=0.15, Maximum=0.35)
     Float SG 1:NiagaraFloat=1
     Float SG 2:NiagaraFloat=1
     Float SG 3:NiagaraFloat=&RandomRangeFloat(Minimum=1, Maximum=1)
     Float SG 4:NiagaraFloat=0
     Float SG 5:NiagaraFloat=0
    )
    Sprite Rotation Mode=::NewEnumerator2
    Sprite Size Mode=::NewEnumerator3
    Uniform Sprite Size=&RandomRangeFloat(Minimum=20, Maximum=30)
    UsePositionOffset=false
    Write Material Random=true
   )
   &PartitionParticles(Number Of Partitions=$Emitter.PortCount)
   &SkeletalMeshLocation(
    Bone / Socket Index=$Particles.PartitionParticles.Partition
    Bone / Socket Sampling Mode=::NewEnumerator5
    MaskSpawnGroup=false
    Mesh Sampling Type=::NewEnumerator6
    Sampled Position Offset=<-40 0 0>
    Skeletal Mesh=$System.SkeletalMesh
   )
   &AddVelocity(
    Cone Angle=&VC_FloatBySpawnGroup(
     SpawnGroup or ExecIndex:NiagaraInt32=0
     CustomInt:NiagaraInt32=0
     MaximumSpawnGroups:NiagaraInt32=3
     Float SG 0:NiagaraFloat=0
     Float SG 1:NiagaraFloat=5
     Float SG 2:NiagaraFloat=5
     Float SG 3:NiagaraFloat=5
     Float SG 4:NiagaraFloat=0
     Float SG 5:NiagaraFloat=0
    )
    Module.UseConeFalloff=false
    Module.UseVelDistribution=false
    Override Seed=false
    Velocity Mode=::NewEnumerator2
    Velocity Speed=&VC_FloatBySpawnGroup(
     SpawnGroup or ExecIndex:NiagaraInt32=0
     CustomInt:NiagaraInt32=0
     MaximumSpawnGroups:NiagaraInt32=3
     Float SG 0:NiagaraFloat=-1500
     Float SG 1:NiagaraFloat=-1000
     Float SG 2:NiagaraFloat=-1000
     Float SG 3:NiagaraFloat=-1000
     Float SG 4:NiagaraFloat=0
     Float SG 5:NiagaraFloat=0
    )
   )
   &ShapeLocation(
    MaskSpawnGroup=false
    Override Seed=false
    Sphere Radius=&VC_FloatBySpawnGroup(
     SpawnGroup or ExecIndex:NiagaraInt32=0
     CustomInt:NiagaraInt32=0
     MaximumSpawnGroups:NiagaraInt32=3
     Float SG 0:NiagaraFloat=0
     Float SG 1:NiagaraFloat=60
     Float SG 2:NiagaraFloat=60
     Float SG 3:NiagaraFloat=40
     Float SG 4:NiagaraFloat=0
     Float SG 5:NiagaraFloat=0
    )
    Sphere Surface Distribution=1
    UseOwnerScale=true
   )
   /SetVariables(Particles.SpriteSize0:Vector2f=<100 150>) {isSetParameters=true, module=SetVariables_#3}
  ]
  particleUpdateScript [
   &ParticleState()
   &ScaleSpriteSize(Uniform Curve Sprite Scale=DataInterface() {propertyValues {Curve {defaultValue=0, keys [{tangentMode=RCTM_User, time=0, value=1}, {tangentMode=RCTM_User, time=1, value=0.5}]}}})
   &SubUVAnimation(
    Random Start Frame=true
    "Sprite Renderer "=DataInterface() {propertyValues {SpriteRenderer=&self:SparksGroup001_0.NiagaraSpriteRendererProperties_3}}
    UseEndFrame=false
    UseStartFrame=false
   )
   &SubUVAnimation(
    "Sprite Renderer "=DataInterface() {propertyValues {SpriteRenderer=&self:SparksGroup001_0.NiagaraSpriteRendererProperties_0}}
    UseEndFrame=false
    UseStartFrame=false
   ) {module=SubUVAnimation001}
   &SubUVAnimation(
    "Sprite Renderer "=DataInterface() {propertyValues {SpriteRenderer=&self:SparksGroup001_0.NiagaraSpriteRendererProperties_5}}
    Start Frame Offset=9
    SubUV Lookup Index Scale:NiagaraFloat=&FloatFromCurve(FloatCurve=DataInterface() {propertyValues {Curve {keys [<at 0 0 RCIM_Cubic>, {arriveTangent=6.1554, interpMode=RCIM_Cubic, leaveTangent=6.1554, tangentMode=RCTM_User, time=1.3, value=3}]}}})
    UseEndFrame=false
    UseStartFrame=false
   ) {module=SubUVAnimation002}
   &SubUVAnimation(
    "Sprite Renderer "=DataInterface() {propertyValues {SpriteRenderer=&self:SparksGroup001_0.NiagaraSpriteRendererProperties_1}}
    Start Frame Offset=9
    SubUV Lookup Index Scale:NiagaraFloat=&FloatFromCurve(FloatCurve=DataInterface() {propertyValues {Curve {keys [<at 0 0 RCIM_Cubic>, {arriveTangent=6.1554, interpMode=RCIM_Cubic, leaveTangent=6.1554, tangentMode=RCTM_User, time=1.3, value=3}]}}})
    UseEndFrame=false
    UseStartFrame=false
   ) {module=SubUVAnimation003}
   /SetVariables(
    Particles.ScaleColor_R1:LinearColor=&VC_DynamicScaleColor(
     Color to Scale:LinearColor=$Particles.R1 Color
     RGB Scale:Vector3f=&VectorFromFloat(Value=3)
     Alpha Scale:NiagaraFloat=1
    )
    Particles.ScaleColor_R3:LinearColor=&VC_DynamicScaleColor(
     Color to Scale:LinearColor=$Particles.Initial.R2 Color
     RGB Scale:Vector3f=&VectorFromFloat(Value=0.2)
     Alpha Scale:NiagaraFloat=1
    )
    Particles.Sprite3b:LinearColor=&VC_DynamicScaleColor(
     Color to Scale:LinearColor=$Particles.Color
     RGB Scale:Vector3f=&VectorFromFloat(Value=3)
     Alpha Scale:NiagaraFloat=&SelectFloatFromArray(
      Array Sampling Mode=::NewEnumerator0
      Direct Array Index=?
      Float Selection Array=DataInterface() {propertyValues {FloatData [0,1]}}
      Random Seed=0
     )
    )
   ) {isSetParameters=true, module=SetVariables_#4}
   &SolveForcesAndVelocity(Clamp Velocity=false, Limit Acceleration=false)
  ]
  rendererClasses [&NiagaraSpriteRendererProperties]
  renderers [
   {rendererClass=&NiagaraSpriteRendererProperties}
   {rendererClass=&NiagaraSpriteRendererProperties, rendererIndex=1}
   {rendererClass=&NiagaraSpriteRendererProperties, rendererIndex=2}
   {rendererClass=&NiagaraSpriteRendererProperties, rendererIndex=3}
  ]
 }
 SparksGroup002 {
  emitterUpdateScript [
   &EmitterState(Inactive Response=::NewEnumerator1, Life Cycle Mode=::NewEnumerator1, Loop Duration=1, UseLoopDelay=false)
   &VC_FilteredSocketCount()
   &SpawnRate(SpawnGroup=1, SpawnRate=10, Use Spawn Probability=false)
   &SpawnRate(SpawnGroup=2, SpawnRate=20, Use Spawn Probability=false) {module=SpawnRate002}
  ]
  particleSpawnScript [
   &VC_PaintedVariantsOverride(
    Black=<rgba 0 0 0 1>
    BurntSienna=<rgba 1 0.2116 0 1>
    Crimson=<rgba 1 0 0.0271 1>
    DefaultColor=<rgba 0 2.2222 5 1>
    ForestGreen=<rgba 0 1 0.0505 1>
    Grey=<rgba 0.009 0.009 0.009 3>
    Lime=<rgba 0.25 0.5 0 1>
    Orange=<rgba 1 0.0833 0 1>
    Pink=<rgba 1 0.1198 0.6799 1>
    Purple=<rgba 0.5 0 1 1>
    Saffron=<rgba 1 0.7667 0 1>
    SkyBlue=<rgba 0 0.6654 1 1>
    TitaniumWhite=<rgba 1 1 1 1>
   )
   /SetVariables(Particles.Visivily_tag:NiagaraInt32=&SpawnGroup()) {isSetParameters=true, module=SetVariables_#2}
   &InitializeParticle(
    Color=$Particles.Color
    Lifetime=&VC_FloatBySpawnGroup(
     SpawnGroup or ExecIndex:NiagaraInt32=0
     CustomInt:NiagaraInt32=0
     MaximumSpawnGroups:NiagaraInt32=3
     Float SG 0:NiagaraFloat=0
     Float SG 1:NiagaraFloat=&RandomRangeFloat(Minimum=0.1, Maximum=0.2)
     Float SG 2:NiagaraFloat=&RandomRangeFloat(Minimum=1, Maximum=1)
     Float SG 3:NiagaraFloat=&RandomRangeFloat(Minimum=1, Maximum=1)
     Float SG 4:NiagaraFloat=0
     Float SG 5:NiagaraFloat=0
    )
    Sprite Rotation Mode=::NewEnumerator2
    UsePositionOffset=false
    Write Material Random=true
   )
   &PartitionParticles(Number Of Partitions=$Emitter.PortCount)
   &SkeletalMeshLocation(
    Bone / Socket Index=$Particles.PartitionParticles.Partition
    Bone / Socket Sampling Mode=::NewEnumerator5
    MaskSpawnGroup=false
    Mesh Sampling Type=::NewEnumerator6
    Sampled Position Offset=<-40 0 0>
    Skeletal Mesh=$System.SkeletalMesh
   )
   &AddVelocity(Velocity=&RandomRangeVector(Minimum=<-500 -300 -300>, Maximum=<-600 300 300>))
   &ShapeLocation(MaskSpawnGroup=false, Override Seed=false, Sphere Radius=30, Sphere Surface Distribution=1, UseOwnerScale=true)
   /SetVariables(
    Particles.Rate2:Vector2f=&RandomRangeVector2D(Minimum=<1 30>, Maximum=<5 50>)
    Particles.rate1:Vector2f=&RandomRangeVector2D(Minimum=<20 80>, Maximum=<40 20>)
   ) {isSetParameters=true, module=SetVariables_#3}
  ]
  particleUpdateScript [
   &ParticleState()
   &SubUVAnimation(
    "Sprite Renderer "=DataInterface() {propertyValues {SpriteRenderer=&self:SparksGroup002_1.NiagaraSpriteRendererProperties_2}}
    UseEndFrame=false
    UseStartFrame=false
   ) {module=SubUVAnimation001}
   /SetVariables(Particles.SpriteSize2=&Multiply_Vector2D(
     A=$Particles.Initial.Rate2
     B=&Vector2DFromCurve(
      Vector2Curve=DataInterface() {propertyValues {XCurve {keys [<at 0.5 1 RCIM_Cubic>,<at 1 0 RCIM_Cubic>]}, YCurve {keys [<at 0 1 RCIM_Cubic>,<at 1 0 RCIM_Cubic>]}}}
      ScaleCurve=<1 1>
     )
    )) {isSetParameters=true, module=SetVariables_#4}
   &Drag(Drag=2, Use Linear Drag=true, Use Rotational Drag=true)
   &SolveForcesAndVelocity(Clamp Velocity=false, Limit Acceleration=false)
   &ScaleColor(Scale RGB=&VectorFromFloat(Value=10), ScaleA=true, ScaleRGB=true)
  ]
  rendererClasses [&NiagaraSpriteRendererProperties]
  renderers [
   {rendererClass=&NiagaraSpriteRendererProperties}
   {rendererClass=&NiagaraSpriteRendererProperties, rendererIndex=1}
   {rendererClass=&NiagaraSpriteRendererProperties, rendererIndex=2}
  ]
 }
}
summary {
 systemName=NS_Boost_Bfooom
 emitters [
  {emitterName=RibbonGroup001, rendererClasses [&NiagaraRibbonRendererProperties]}
  {emitterName=SparksGroup001, rendererClasses [&NiagaraSpriteRendererProperties]}
  {emitterName=SparksGroup002, rendererClasses [&NiagaraSpriteRendererProperties]}
 ]
 userVariables [
  User.bIsFrontendPreview {type=NiagaraBool}
  User.bIsPainted {type=NiagaraBool}
  User.ColorVariantID {type=NiagaraInt32}
  User.DTCarOffset {type=Vector3f, description="", defaultValue {struct=Vector3f, value=<0 0 0>}}
  User.DTCarOffset001 {type=Vector3f, description="", defaultValue {struct=Vector3f, value=<0 0 0>}}
  User.PaintedColor {type=LinearColor, description="", defaultValue {struct=LinearColor, value=<rgba 1 1 1 1>}}
  User.PaintedScalar {type=NiagaraFloat, defaultValue=0, description=""}
  User.PortCount {type=NiagaraInt32, defaultValue=2}
  User.SparksColorScale {type=NiagaraFloat, defaultValue=2, description=""}
 ]
}
