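# Translucent unlit master material for the core trail of the Bfooom vehicle boost effect, usable on Niagara ribbons and sprites.
# Layers panning noise textures with distorted UVs, tiled by scalar or dynamic parameters (TEX_U drives both axes).
# The noise modulates particle-color brightness for an emissive-adjustment function and boosts a UV-gradient edge mask, optionally power-shaped and scaled by particle alpha.
# A camera-offset function feeds a view-space depth offset vector.
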Material /VehicleCosmetics/Boosts/Boost_Bfooom/Material/M_Boost_Bfoom_CoreTrail_Master_Mat
graph {
 properties {
  bAutomaticallySetUsageInEditor=false
  bUsedWithNiagaraRibbons=true
  bUsedWithNiagaraSprites=true
  blendMode=BLEND_Translucent
  refractionMethod=RM_PixelNormalOffset
  shadingModel=MSM_Unlit
  usedShadingModels=Unlit
 }
 nodes [
  n1 = TextureSample(UVs=Panner(Coordinate=TextureCoordinate() {coordinateIndex=1, vTiling=0.5}) {speedX=-0.5, speedY=0.25}) {texture=/VehicleCosmetics/Boosts/Boost_Bfooom/Texture/T_VCVFX_DistortNoise_RGB_T}
  n2 = TextureSample(UVs=Panner(Coordinate=TextureCoordinate() {coordinateIndex=1, vTiling=-0.5}) {speedX=-1, speedY=0.25}) {texture=&T_VCVFX_DistortNoise_RGB_T}
  n3 = Add(A=TextureCoordinate() {coordinateIndex=1}, B=Multiply(A=AppendVector(A=@n1.R, B=@n2.R), B=Constant() {r=0.5}))
  n4 = DynamicParameter() {paramNames [TEX_U,TEX_V,Param3,Param4]}
  n5 = AppendVector(A=@n4.TEX_U, B=@n4.TEX_U)
  n6 = If(
   A=ScalarParameter() {parameterName=Use Dynamic Tiling}
   "A > B"=@n5
   "A == B"=@n5
   "A < B"=AppendVector(A=ScalarParameter() {defaultValue=1, parameterName=Tex_U}, B=ScalarParameter() {defaultValue=1, parameterName=Tex_V})
  )
  n7 = TextureSample(UVs=Panner(Coordinate=Multiply(A=@n3, B=@n6)) {speedX=-2, speedY=0.25}) {texture=/DealershipSharedLibrary/FX/Textures/Noise/T_VCVFX_DistortNoise02_RGB}
  n8 = TextureCoordinate() {coordinateIndex=1, vTiling=-1}
  n9 = TextureSample(UVs=Panner(Coordinate=Add(A=@n2.R, B=@n8)) {speedX=-0.5, speedY=0.25}) {texture=&T_VCVFX_DistortNoise02_RGB}
  n10 = Add(A=@n7.G, B=@n9.R)
  n11 = Multiply(A=Add(A=@n7.R, B=@n9.G), B=Add(A=@n7.R, B=@n9.R))
  n12 = Add(A=@n7.B, B=@n9.B)
  AppendVector(A=AppendVector(A=Multiply(A=@n10, B=@n9.B), B=@n11), B=Add(A=@n10, B=@n12))
  MaterialExpressionVertexColor_0 = ParticleColor()
  n13 = TextureCoordinate() {vTiling=0.85}
  n14 = Constant() {r=30}
  n15 = OneMinus(None=ComponentMask(None=@n13) {r=true})
  MaterialExpressionConstantClamp_2 = Saturate(None=Multiply(A=ComponentMask(None=@n13) {r=true}, B=Multiply(A=@n14, B=@n15)))
  n16 = ComponentMask(None=@n13) {g=true}
  n17 = Multiply(A=@n16, B=OneMinus(None=@n16))
  n18 = Constant() {r=5}
  n19 = Multiply(A=@MaterialExpressionConstantClamp_2, B=Multiply(A=OneMinus(None=@n16), B=Multiply(A=@n17, B=@n18)))
  MaterialExpressionConstantClamp_1 = Saturate(None=Subtract(A=@n12, B=@n11))
  n20 = Add(A=@n19, B=Multiply(A=@n19, B=@MaterialExpressionConstantClamp_1))
  n21 = Power(Base=@n20, Exp=ScalarParameter() {parameterName=Power_Exp})
  n22 = ScalarParameter() {parameterName=Use Power on Edge Mask}
  MaterialExpressionConstantClamp_0 = Saturate(None=Multiply(A=@MaterialExpressionVertexColor_0.A, B=LinearInterpolate(A=@n20, B=@n21, Alpha=@n22)))
  n23 = Constant() {r=0.1}
  MaterialFunctionCall(Emissive=LinearInterpolate(A=Multiply(A=@n23, B=@MaterialExpressionVertexColor_0.RGB), B=@MaterialExpressionVertexColor_0.RGB, Alpha=@MaterialExpressionConstantClamp_1)) {materialFunction=&MF_FX_EmissiveAdjustments}
  n24 = Constant3Vector() {constant=<rgba 0 0 1 0>}
  StaticSwitchParameter(True=@n24) {defaultValue=true, parameterName=IS MASKED ?}
  n25 = Constant() {r=-20}
  n26 = MaterialFunctionCall(Offset Amount=@n25, Clamp Padding=@n25) {materialFunction=/Engine/Functions/Engine_MaterialFunctions02/WorldPositionOffset/CameraOffset}
  n27 = ScalarParameter()
  Transform(None=Multiply(A=Add(A=@n26.Clamped Camera Offset, B=@n27), B=@n24)) {transformSourceType=TRANSFORMSOURCE_View}
 ]
}
