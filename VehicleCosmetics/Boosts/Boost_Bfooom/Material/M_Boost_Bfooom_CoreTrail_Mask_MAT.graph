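# Translucent unlit material for the core trail mask of the Bfooom vehicle boost effect, usable on Niagara ribbons, sprites and mesh particles.
# Cloud noise warps a panning liquid-trail normal map, whose desaturated first sample drives BumpOffset parallax for a second sample.
# That result, UV edge falloffs and particle alpha form a saturated mask.
# Side branches build layered distortion noise, squared particle colour, an "IS MASKED ?" switch and a camera offset.
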Material /VehicleCosmetics/Boosts/Boost_Bfooom/Material/M_Boost_Bfooom_CoreTrail_Mask_MAT
graph {
 properties {
  bAutomaticallySetUsageInEditor=false
  bUsedWithMeshParticles=true
  bUsedWithNiagaraRibbons=true
  bUsedWithNiagaraSprites=true
  bUsedWithParticleSprites=true
  blendMode=BLEND_Translucent
  refractionMethod=RM_PixelNormalOffset
  shadingModel=MSM_Unlit
  usedShadingModels=Unlit
 }
 nodes [
  n1 = TextureCoordinate() {coordinateIndex=1}
  n2 = Panner(Coordinate=TextureCoordinate() {coordinateIndex=1, vTiling=0.5}) {speedX=-0.5, speedY=-0.5}
  n3 = Panner(Coordinate=TextureCoordinate() {coordinateIndex=1, vTiling=-0.5}) {speedX=-2, speedY=-0.5}
  n4 = Multiply(
   A=AppendVector(
    A=TextureSample(UVs=@n2) {texture=/VehicleCosmetics/Boosts/Boost_Bfooom/Texture/T_VCVFX_DistortNoise_RGB_T}.G
    B=TextureSample(UVs=@n3) {texture=&T_VCVFX_DistortNoise_RGB_T}.R
   )
   B=Constant() {r=0.5}
  )
  n5 = DynamicParameter() {paramNames [UVDistort,PanSpeed,Param3,Param4]}
  n6 = Time()
  n7 = TextureSample(UVs=Panner(Coordinate=Add(A=@n1, B=@n4), Time=Multiply(A=@n5.UVDistort, B=@n6)) {speedX=-1, speedY=0.25}) {texture=/DealershipSharedLibrary/FX/Textures/Noise/T_VCVFX_DistortNoise02_RGB}
  n8 = TextureCoordinate() {coordinateIndex=1, vTiling=-1}
  n9 = TextureSample(UVs=Panner(Coordinate=Add(A=@n4, B=@n8)) {speedX=-0.25, speedY=0.25}) {texture=&T_VCVFX_DistortNoise02_RGB}
  n10 = Add(A=@n7.R, B=@n9.R)
  n11 = Add(A=@n7.B, B=@n9.B)
  n12 = Add(A=@n7.G, B=@n9.R)
  AppendVector(A=AppendVector(A=Multiply(A=@n10, B=@n9.B), B=Multiply(A=@n11, B=@n12)), B=Add(A=@n11, B=@n12))
  n13 = ParticleColor()
  MaterialFunctionCall(Emissive=Multiply(A=@n13.RGB, B=@n13.RGB)) {materialFunction=&MF_FX_EmissiveAdjustments}
  n14 = Constant() {r=0.8}
  n15 = TextureCoordinate()
  n16 = ComponentMask(None=@n15) {r=true}
  n17 = ComponentMask(None=@n15) {g=true}
  n18 = Multiply(A=@n17, B=OneMinus(None=@n17))
  n19 = Constant() {r=1}
  n20 = Multiply(A=Saturate(None=Multiply(A=@n14, B=@n16)), B=Multiply(A=OneMinus(None=@n16), B=Multiply(A=@n18, B=@n19)))
  n21 = LinearInterpolate(
   A=TextureCoordinate() {coordinateIndex=1, uTiling=0.5, vTiling=0.25}
   B=TextureCoordinate() {coordinateIndex=1, uTiling=0.35, vTiling=0.25}
   Alpha=Sine(None=Time())
  )
  n22 = Constant() {r=0.5}
  n23 = TextureCoordinate() {coordinateIndex=1, vTiling=0.25}
  n24 = TextureCoordinate() {coordinateIndex=1, uTiling=0.5, vTiling=0.25}
  n25 = AppendVector(
   A=TextureSample(UVs=Panner(Coordinate=@n23) {speedX=-1, speedY=0.15}) {texture=/DealershipSharedLibrary/FX/Textures/Noise/T_VCVFX_CloudsSoft_RGB}.G
   B=TextureSample(UVs=Panner(Coordinate=@n24) {speedX=-0.8, speedY=-0.1}) {texture=&T_VCVFX_CloudsSoft_RGB}.R
  )
  n26 = Multiply(A=Constant() {r=0.05}, B=@n25)
  n27 = Add(A=@n21, B=@n26)
  n28 = Constant2Vector() {g=0.1}
  n29 = TextureSample(UVs=Panner(Coordinate=Add(A=@n27, B=@n28)) {speedX=-0.5}) {samplerType=SAMPLERTYPE_Normal, texture=/VehicleCosmetics/Boosts/Boost_Bfooom/Texture/T_VCVFX_LiquidTrail_N}
  n30 = Saturate(None=ConstantBiasScale(None=Desaturation(None=@n29.RGB) {luminanceFactors=<rgba 0.3 0.59 0.11 0>}) {bias=-0.3, scale=4})
  n31 = Add(A=BumpOffset(Coordinate=@n21, Height=Multiply(A=@n22, B=@n30), HeightRatioInput=Constant() {r=0.5}), B=@n26)
  n32 = Constant2Vector() {g=0.3}
  n33 = Constant() {r=0.5}
  n34 = Time()
  n35 = TextureSample(UVs=Panner(Coordinate=Add(A=@n31, B=@n32), Time=Multiply(A=@n33, B=@n34)) {speedX=-5}) {samplerType=SAMPLERTYPE_Normal, texture=&T_VCVFX_LiquidTrail_N}
  n36 = Saturate(None=ConstantBiasScale(None=Desaturation(None=@n35.RGB) {luminanceFactors=<rgba 0.3 0.59 0.11 0>}) {bias=-0.01, scale=2})
  n37 = TextureCoordinate()
  n38 = Multiply(A=@n25, B=Constant() {r=0.38})
  n39 = ComponentMask(None=Add(A=Add(A=@n37, B=@n38), B=Constant2Vector() {g=-0.12})) {g=true}
  n40 = Multiply(A=Power(Base=OneMinus(None=@n39), Exp=Constant() {r=2}), B=Constant() {r=3})
  n41 = Saturate(None=ConstantBiasScale(None=Multiply(A=@n39, B=@n40)) {bias=0, scale=20})
  n42 = Multiply(A=Multiply(A=Multiply(A=@n36, B=@n41), B=@n41), B=Constant() {r=5})
  Saturate(None=Multiply(A=@n13.A, B=Multiply(A=@n20, B=@n42)))
  n43 = Constant3Vector() {constant=<rgba 0 0 1 0>}
  StaticSwitchParameter(True=@n43) {defaultValue=true, parameterName=IS MASKED ?}
  n44 = Constant() {r=-20}
  n45 = MaterialFunctionCall(Offset Amount=@n44, Clamp Padding=@n44) {materialFunction=/Engine/Functions/Engine_MaterialFunctions02/WorldPositionOffset/CameraOffset}
  n46 = ScalarParameter()
  Transform(None=Multiply(A=Add(A=@n45.Clamped Camera Offset, B=@n46), B=@n43)) {transformSourceType=TRANSFORMSOURCE_View}
 ]
}
