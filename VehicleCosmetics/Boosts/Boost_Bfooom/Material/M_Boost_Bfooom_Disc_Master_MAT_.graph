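# Unlit translucent master material for the toon disc particle effect of the Bfooom vehicle boost cosmetic.
# Dynamic parameters set brightness and two switches, snapped by If/Ceil/Floor, that pick plain or inverted alpha-masked disc texture and darkened or full particle color.
# The texture blends darkened and chosen color, scaled by brightness into the emissive adjustments function.
# Particle alpha multiplies texture RGB, which the texture switch leaves unchanged because both lerp inputs match.
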Material /VehicleCosmetics/Boosts/Boost_Bfooom/Material/M_Boost_Bfooom_Disc_Master_MAT_
graph {
 properties {
  bAutomaticallySetUsageInEditor=false
  bUsedWithBeamTrails=true
  bUsedWithMeshDeformer=true
  bUsedWithNiagaraRibbons=true
  bUsedWithNiagaraSprites=true
  bUsedWithParticleSprites=true
  bUsedWithSkeletalMesh=true
  blendMode=BLEND_Translucent
  shadingModel=MSM_Unlit
  usedShadingModels=Unlit
 }
 nodes [
  MaterialExpressionVertexColor_4 = ParticleColor()
  n1 = Multiply(A=Add(A=@MaterialExpressionVertexColor_4.RGB, B=Constant() {r=-0.3}), B=Constant() {r=0.1})
  LinearInterpolate(A=@n1, B=@MaterialExpressionVertexColor_4.RGB, Alpha=Constant() {r=0.35})
  n2 = DynamicParameter() {paramNames [Brightness, TextureSwitch, ColorSwitch, Param4]}
  MaterialExpressionConstantClamp_11 = Saturate(None=@n1)
  n3 = Ceil(None=@n2.ColorSwitch)
  n4 = LinearInterpolate(
   A=@MaterialExpressionConstantClamp_11
   B=@MaterialExpressionVertexColor_4.RGB
   Alpha=If(A=@n2.ColorSwitch, B=Constant() {r=0.49}, "A > B"=@n3, "A == B"=@n3, "A < B"=Floor(None=@n2.ColorSwitch))
  )
  n5 = TextureSample() {texture=/VehicleCosmetics/Boosts/Boost_Bfooom/Texture/T_VCVFX_ToonDisc_FB_T}
  MaterialExpressionConstantClamp_12 = Saturate(None=Multiply(A=OneMinus(None=@n5.RGB), B=@n5.A))
  n6 = Ceil(None=@n2.TextureSwitch)
  n7 = If(A=@n2.TextureSwitch, B=Constant() {r=0.49}, "A > B"=@n6, "A == B"=@n6, "A < B"=Floor(None=@n2.TextureSwitch))
  n8 = LinearInterpolate(A=@n5.RGB, B=@MaterialExpressionConstantClamp_12, Alpha=@n7)
  MaterialFunctionCall(Emissive=Multiply(A=@n2.Brightness, B=LinearInterpolate(A=@MaterialExpressionConstantClamp_11, B=@n4, Alpha=@n8))) {materialFunction=&MF_FX_EmissiveAdjustments}
  Multiply(A=@MaterialExpressionVertexColor_4.A, B=LinearInterpolate(A=@n5.RGB, B=@n5.RGB, Alpha=@n7))
 ]
}
