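# Translucent unlit master material for lightning particles in the Bfooom vehicle boost, usable on sprites and Niagara ribbons.
# A lightning texture, sampled as a time-driven flipbook or static optionally rotated UVs, yields a masked channel or a multiply, add or lerp channel blend.
# An optional panning mask modulates that shape.
# Dynamic parameters drive brightness and blend, while particle color feeds emissive adjustments and depth-faded opacity.
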Material /VehicleCosmetics/Boosts/Boost_Bfooom/Material/M_Boost_Bfooom_Lightning_Master_MAT
graph {
 properties {
  bAutomaticallySetUsageInEditor=false
  bUsedWithNiagaraRibbons=true
  bUsedWithNiagaraSprites=true
  bUsedWithParticleSprites=true
  blendMode=BLEND_Translucent
  shadingModel=MSM_Unlit
  usedShadingModels=Unlit
 }
 nodes [
  n1 = Multiply(A=Time(), B=ScalarParameter() {defaultValue=1, group=Pan, parameterName=PanSpeed})
  n2 = ComponentMask(None=Panner(Coordinate=TextureCoordinate() {uTiling=0.25, vTiling=0.05}, Time=@n1) {speedX=-0.4}) {r=true}
  n3 = TextureCoordinate() {uTiling=0.05, vTiling=0.25}
  n4 = AppendVector(A=@n2, B=ComponentMask(None=Panner(Coordinate=@n3, Time=@n1) {speedY=-0.4}) {g=true})
  n5 = TextureSampleParameter2D(UVs=StaticComponentMaskParameter(None=AppendVector(A=@n4, B=@n2)) {defaultB=true, group=Pan, parameterName=RGB_Pan_Mask}) {group=Pan, parameterName=PanMask, texture=/VehicleCosmetics/Boosts/Boost_Bfooom/Texture/T_VCVFX_Time_RGB_M_T}
  n6 = ScalarParameter() {defaultValue=2, group=FlipBook, parameterName=Columns}
  n7 = ScalarParameter() {defaultValue=4, group=FlipBook, parameterName=Rows}
  n8 = AppendVector(A=@n6, B=@n7)
  n9 = Divide(
   A=MaterialFunctionCall(UVs=TextureCoordinate(), Texture Scale=VectorParameter() {parameterName=UVScale}.RGB) {materialFunction=&ScaleUVsByCenter}.UVs
   B=@n8
  )
  n10 = AppendVector(A=Multiply(A=@n6, B=@n7), B=@n7)
  n11 = Frac(None=Multiply(A=Time(), B=ScalarParameter() {defaultValue=1, group=FlipBook, parameterName=Speed}))
  n12 = AppendVector(A=@n11, B=@n11)
  n13 = Divide(A=Floor(None=Multiply(A=@n10, B=@n12)), B=@n8)
  n14 = TextureCoordinate() {coordinateIndex=1}
  n15 = Rotator(Coordinate=@n14, Time=Constant() {r=0.5}) {speed=3.14}
  n16 = TextureSampleParameter2D(UVs=StaticSwitchParameter(True=Add(A=@n9, B=@n13), False=StaticSwitchParameter(True=@n15, False=@n14) {group=FlipBook, parameterName=RotateFlipbook}) {defaultValue=true, group=FlipBook, parameterName=FlipBook}) {group=FlipBook, parameterName=BaseTexture, texture=/VehicleCosmetics/Boosts/Boost_Bfooom/Texture/T_VCVFX_LightningRGB_02}
  n17 = StaticComponentMaskParameter(None=@n16.RGB) {defaultG=true, group=Blend, parameterName=Blend_01}
  n18 = StaticComponentMaskParameter(None=@n16.RGB) {defaultB=true, group=Blend, parameterName=Blend_02}
  n19 = Multiply(A=@n17, B=@n18)
  n20 = Add(A=@n17, B=@n18)
  n21 = DynamicParameter() {paramNames [Brightness,LerpBlend,Param3,Param4]}
  n22 = StaticSwitchParameter(
   True=StaticSwitchParameter(
    True=StaticSwitchParameter(True=@n19, False=@n20) {group=Blend, parameterName="BlendType(T=M)(F=A)"}
    False=LinearInterpolate(A=@n17, B=@n18, Alpha=@n21.LerpBlend)
   ) {group=Blend, parameterName=LerpBlend}
   False=StaticComponentMaskParameter(None=@n16.RGB) {defaultG=true, group=Blend, parameterName=RGB_Mask}
  ) {group=Blend, parameterName=Blend}
  n23 = Multiply(
   A=Multiply(
    A=ScalarParameter() {defaultValue=3, group=Pan, parameterName=PanBrightness}
    B=StaticComponentMaskParameter(None=@n5.RGB) {defaultB=true, group=Pan, parameterName=RGB_Pan_Mask}
   )
   B=@n22
  )
  n24 = StaticSwitchParameter(True=StaticSwitchParameter(True=Add(A=@n23, B=@n22), False=@n23) {group=Pan, parameterName=PanAdd}, False=@n22) {group=Pan, parameterName=PanLightning}
  n25 = ParticleColor()
  MaterialFunctionCall(Emissive=Multiply(A=Multiply(A=@n24, B=@n21.Brightness), B=@n25.RGB)) {materialFunction=&MF_FX_EmissiveAdjustments}
  n26 = Multiply(A=@n24, B=Constant() {r=1})
  Saturate(None=MaterialFunctionCall(Opacity In=Multiply(A=@n25.A, B=@n26)) {materialFunction=&MF_VCVFX_DepthFade}.Result)
 ]
}
